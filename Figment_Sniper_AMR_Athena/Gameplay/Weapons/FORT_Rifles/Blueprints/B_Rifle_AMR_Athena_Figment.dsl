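; Weapon Blueprint for the Figment chargeable AMR sniper rifle, driving its charge, fire, reload and out-of-ammo audio and visual effects.
; Charge events swap beam, muzzle and glow effects, with max charge raising local pierce-through and looping camera shake.
; Timelines ramp arc particles, audio pitch and low-ammo mesh slowdown.
; Release, firing, unequip, and DBNO or holster delegates end charge effects to prevent lingering sound.
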(event EventMuzzlePlayReloadFX
  (params (Selection :type EFortReloadFXState%20Enum :default "ReloadStart"))
  (|CallParentFunction Selection))

(event EventOnPlayReloadFX
  (params (ReloadStage :type EFortReloadFXState%20Enum :default "ReloadStart"))
  ;[1 1 1] Reload
  (|CallParentFunction ReloadStage)
  (switch Utilities|FlowControl|Switch|SwitchonEFortReloadFXState ReloadStage
    (:ReloadStart
      (CallFunction|EndOfCharge)
      (Utilities|FlowControl|DoOnce false
        (:Completed
          ; Charge audio bug fixing
          (Variables|Default|SetReloading true)
          (Utilities|FlowControl|Delay 1.0))))
    (:ReloadEnd
      (Components|Animation|PlayAnimation (Variables|Visual|GetWeaponMesh) "/Figment_Sniper_AMR_Athena/Gameplay/Animation/Game/MainPlayer/Combat/Guns/AssaultRifle/AMR_SpinUP.AMR_SpinUP" true)
      (Components|Activation|Activate (Variables|Default|GetCoreParticles) true)
      (Utilities|FlowControl|Delay 0.8)
      (Variables|Default|SetReloading false))))

(event EventOnAmmoCountChanged
  ;[1 1 1] Ammo check
  (if (>= (FortWeapon|GetLocalMagazineAmmoCount) (Variables|Default|GetAmmoCost_MinCharge))
    (Variables|VFX|SetOutOfAmmo_Local false)
    (Components|Animation|SetPlayRate (Variables|Visual|GetWeaponMesh) (Variables|VFX|GetAnimationRate(Uncharged)))
    (Components|Activation|Activate (Variables|Default|GetCoreParticles) false)
    (else
      (Variables|VFX|SetOutOfAmmo_Local true)
      (Components|Activation|Deactivate (Variables|Default|GetCoreParticles))
      (Variables|VFX|SetCurrentPlayrate (Components|Animation|GetPlayRate (Variables|Visual|GetWeaponMesh)))
      (bind (_direction _lerp) (Timeline|SlowdownTimeline PlayFromStart)
        (:Update
          (Components|Animation|SetPlayRate (Variables|Visual|GetWeaponMesh) (Math|Float|Lerp 0.0 (Variables|VFX|GetCurrentPlayrate) _lerp)))))))

(event EventOnStartCharge
  (bind _ambient_charge_audiocomponent (Variables|Default|GetAmbient_Charge_AudioComponent))
  (Variables|Default|GetReloading
    (:then)
    (:else
      ;[1 1 1] Start Charing
      (Variables|VFX|SetCharging true)
      (Variables|VFX|SetReached_Min_Charge false) ; Can play charged weapon fire sound
      (CallFunction|VFXSetUnchargedParams)
      ;;[1 1 1] Weapon charge audio setup
      (Audio|Components|Audio|FadeOut (Variables|Default|GetAmbient_Charge_AudioComponent) 0.3 0.0 "Linear" :!class /Script/Engine.AudioComponent)
      (bind _returnvalue (Audio|SpawnSoundAttached "/Game/Sounds/Fort_Weapons/Rifle/Weapon_RFL_AMR_T04_Epic_AMR_Charge_Cue.Weapon_RFL_AMR_T04_Epic_AMR_Charge_Cue" (Variables|Visual|GetWeaponMesh)))
      (Variables|Default|SetChargeAudio _returnvalue)
      (Audio|Components|Audio|Play (Variables|Default|GetFullCharge_AudioComponent) 0.0)
      (Utilities|FlowControl|Sequence
        (:then_0
          (bind (_direction _reload_pitch) (Timeline|Reload_PItchChange Play)
            (:Update
              (Audio|Components|Audio|SetPitchMultiplier _ambient_charge_audiocomponent _reload_pitch)
              (Audio|Components|Audio|SetVolumeMultiplier _ambient_charge_audiocomponent _reload_pitch :!class /Script/Engine.AudioComponent))))
        (:then_1
          ; FORT-850921 - stops charge SFX when entering DBNO
          (bind _asfort_player_pawn (Utilities|Casting|CastToFortPlayerPawn (Game|GetInstigator)))
          (EventDispatchers|BindEventtoOnPlayerStartDBNO _asfort_player_pawn (delegate Custom|PlayerStartDBNO)))
        (:then_2
          (bind _asfort_player_pawn_2 (Utilities|Casting|CastToFortPlayerPawn (Game|GetInstigator)))
          (EventDispatchers|BindEventtoOnHolsteredEvent _asfort_player_pawn_2 (delegate Custom|OnHolstered)))))))

(event Custom|PlayerStartDBNO
  ; FORT-850921 - stops charge SFX when entering DBNO
  (CallFunction|EndOfCharge))

(event Custom|OnHolstered
  (CallFunction|EndOfCharge))

(event EventOnLocalClientReachedMaxCharge
  (bind _residualarcstop1 (Variables|Default|GetResidualArcsTop1))
  ;[1 1 1] Max Charge Reached
  (CallFunction|VFXSetChargedParams)
  (Variables|VFX|SetReached_Min_Charge true) ; Can play charged weapon fire sound
  (CallFunction|SetWeaponPierceThrough true 12)
  (Components|Activation|SetActive _residualarcstop1 true true)
  (Utilities|FlowControl|Sequence
    (:then_0
      (CallFunction|ChargedFXOn))
    (:then_1
      (bind (_direction _spawnrate) (Timeline|ScaleUpParticles PlayFromStart)
        (:Update
          (Effects|Components|ParticleSystem|SetFloatParameter _residualarcstop1 "SpawnRate" _spawnrate))))
    (:then_2
      (Camera|PlayWorldCameraShake "/Figment_Sniper_AMR_Athena/Gameplay/Weapons/FORT_Rifles/Effects/Muzzle/B_AMR_Shake_FullyCharged_01.B_AMR_Shake_FullyCharged_01_C" (Transformation|GetActorLocation) 0.0 200)
      (Utilities|FlowControl|Delay 0.5)
      (if (Variables|VFX|GetCharging)
        (CallFunction|CameraShakeLoop)))))

(event EventOnRemoteClientReachedMaxCharge
  (bind _residualarcstop1 (Variables|Default|GetResidualArcsTop1))
  (CallFunction|VFXSetChargedParams)
  (Components|Activation|SetActive _residualarcstop1 true true)
  (Utilities|FlowControl|Sequence
    (:then_0
      (CallFunction|ChargedFXOn))
    (:then_1
      (bind (_direction _spawnrate) (Timeline|ScaleUpParticles_Remote PlayFromStart)
        (:Update
          (Effects|Components|ParticleSystem|SetFloatParameter _residualarcstop1 "SpawnRate" _spawnrate))))))

(event Custom|CameraShakeLoop
  ;[1 1 1] Max Charge Reached
  (Camera|PlayWorldCameraShake "/Figment_Sniper_AMR_Athena/Gameplay/Weapons/FORT_Rifles/Effects/Muzzle/B_AMR_Shake_FullyCharged_01.B_AMR_Shake_FullyCharged_01_C" (Transformation|GetActorLocation) 0.0 200)
  (Utilities|FlowControl|Delay 0.5)
  (if (Variables|VFX|GetCharging)
    (CallFunction|CameraShakeLoop)))

(event Custom|EndOfCharge
  ;[1 1 1] Max Charge Reached
  (bind _chargeaudio (Variables|Default|GetChargeAudio)
    (:then
      (Audio|Components|Audio|FadeOut _chargeaudio 0.2 0.0 "Linear" :!class /Script/Engine.AudioComponent))
    (:else
      _))
  (Audio|Components|Audio|FadeOut (Variables|Default|GetFullCharge_AudioComponent) 0.2 0.0 "Linear" :!class /Script/Engine.AudioComponent)
  (Components|Timeline|Stop (Variables|B_Rifle_AMR_Athena_Figment|GetScaleUpParticles_Remote))
  (CallFunction|ChargedFXOff))

(event EventOnChargeDown
  ;[1 1 1] Charge Down (mouse released)
  (CallFunction|ChargedFXOff)
  (CallFunction|EndOfCharge)
  (Utilities|FlowControl|Delay 0.034)
  (CallFunction|SetWeaponPierceThrough true 1)
  (Variables|VFX|SetCharging false))

(event Custom|Evemt%20ChargeDown
  (params (Weapon :type Fort%20Weapon%20Object%20Reference))
  ;[1 1 1] Charge Down (mouse released)
  (CallFunction|ChargedFXOff)
  (CallFunction|EndOfCharge)
  (Utilities|FlowControl|Delay 0.034)
  (CallFunction|SetWeaponPierceThrough true 1)
  (Variables|VFX|SetCharging false))

(event EventOnPlayWeaponFireFX
  (params (bPersistentFire :type Boolean) (bSecondaryFire :type Boolean))
  (bind _ambient_charge_audiocomponent (Variables|Default|GetAmbient_Charge_AudioComponent))
  (bind _residualarcstop1 (Variables|Default|GetResidualArcsTop1))
  ;[1 1 1] Event On Fire Stuff
  (|CallParentFunction bPersistentFire false)
  (Components|Activation|Activate (Variables|Default|GetMuzzleLight) false) ; Toggles Muzzle Flash Light
  (if (Variables|VFX|GetReached_Min_Charge)
    (Utilities|FlowControl|Sequence
      (:then_0
        (CallFunction|EndOfCharge)
        (Components|Activation|SetActive _residualarcstop1 true true))
      (:then_1
        (bind (_direction _spawnrate) (Timeline|ScaleDownParticles PlayFromStart)
          (:Update
            (Effects|Components|ParticleSystem|SetFloatParameter (Variables|Default|GetResidualArcsSide1) "SpawnRate" _spawnrate))
          (:Finished
            (Components|Activation|SetActive _residualarcstop1 false true)))))
    (else
      _))
  ;[1 1 1] Fire Sound
  ;;[1 1 1] Reset ambient volume and pitch
  (Audio|Components|Audio|SetPitchMultiplier _ambient_charge_audiocomponent 1)
  (Audio|Components|Audio|SetVolumeMultiplier _ambient_charge_audiocomponent 0.0 :!class /Script/Engine.AudioComponent)
  (Audio|Components|Audio|Stop (Variables|Default|GetFullCharge_AudioComponent))
  ;;[1 1 1] If charged, play charged fire sound layer
  (if (Variables|VFX|GetReached_Min_Charge)
    (Audio|SpawnSoundAttached "/Figment_Sniper_AMR_Athena/Gameplay/STW_Sounds/Sounds/Fort_Weapons/Rifle/Weapon_RFL_AMR_T04_Epic_Anti_Material_RifleCH_Cue.Weapon_RFL_AMR_T04_Epic_Anti_Material_RifleCH_Cue" _ambient_charge_audiocomponent) ; Play Charged Shot Layer
    (Variables|VFX|SetReached_Min_Charge false))
  (CallFunction|EvemtChargeDown 0))

(event EventBeginPlay
  ;[1 1 1] Begin Play
  (Networking|SwitchHasAuthority
    (:Remote
      (Components|Animation|PlayAnimation (Variables|Visual|GetWeaponMesh) "/Figment_Sniper_AMR_Athena/Gameplay/Animation/Game/MainPlayer/Combat/Guns/AssaultRifle/AMR_SpinUP.AMR_SpinUP" true))))

(event FortWeapon|EventOnLocalUnEquip
  ;[1 1 1] On Unequip - to fix a bug if you were charging, then switched weapons before firing
  (Variables|VFX|SetCharging false)
  (CallFunction|VFXSetUnchargedParams)
  (Components|Animation|SetPlayRate (Variables|Visual|GetWeaponMesh) (Variables|VFX|GetAnimationRate(Uncharged)))
  (Components|Activation|SetActive (Variables|Default|GetResidualArcsTop1) false false)
  (Effects|Components|ParticleSystem|SetFloatParameter (Variables|Default|GetElectrodeParticles) "SpawnRate" (Variables|VFX|GetElectrodeParticlesEnergizedSpawnRate(Uncharged)))
  (CallFunction|EndOfCharge)
  (Components|Timeline|SetNewTime (Variables|B_Rifle_AMR_Athena_Figment|GetReload_PItchChange) 0.0)
  (CallFunction|EvemtChargeDown 0))

(event FortWeapon|EventOnUnEquippedAndDetached
  ;[1 1 1] On Unequip - to fix a bug if you were charging, then switched weapons before firing
  (Variables|VFX|SetCharging false)
  (CallFunction|VFXSetUnchargedParams)
  (Components|Animation|SetPlayRate (Variables|Visual|GetWeaponMesh) (Variables|VFX|GetAnimationRate(Uncharged)))
  (Components|Activation|SetActive (Variables|Default|GetResidualArcsTop1) false false)
  (Effects|Components|ParticleSystem|SetFloatParameter (Variables|Default|GetElectrodeParticles) "SpawnRate" (Variables|VFX|GetElectrodeParticlesEnergizedSpawnRate(Uncharged)))
  (CallFunction|EndOfCharge)
  (Components|Timeline|SetNewTime (Variables|B_Rifle_AMR_Athena_Figment|GetReload_PItchChange) 0.0)
  (CallFunction|EvemtChargeDown 0))

(event FortWeapon|EventOnEquippedAndAttached
  ;[1 1 1] Stops Spring Controller from wigging out on first equip
  (|CallParentFunction))

(fn ChargedFXOff
  (bind _turbinespinrate_uncharged (Variables|VFX|GetTurbineSpinRate(Uncharged)))
  (Transformation|AddLocalRotation (Variables|Default|GetAMRFan) (Math|Rotator|MakeRotator 0.0 _turbinespinrate_uncharged) false false)
  (Transformation|AddLocalRotation (Variables|Default|GetCoreParticles) (Math|Rotator|MakeRotator _turbinespinrate_uncharged) false false)
  (Effects|Components|ParticleSystem|SetFloatParameter (Variables|Default|GetCoreParticles) "SpawnRate" (Variables|VFX|GetCoreParticlesSpawnRate(Uncharged)))
  (Variables|/Script/CascadeRuntime.ParticleSystemComponent|SetCustomTimeDilation (Variables|VFX|GetCoreParticlesTimeDilation(Uncharged)) (Variables|Default|GetCoreParticles))
  (Effects|Components|ParticleSystem|SetFloatParameter (Variables|Default|GetElectrodeParticles) "SpawnRate" (Variables|VFX|GetElectrodeParticlesEnergizedSpawnRate(Uncharged)))
  (Rendering|Material|SetVectorParameterValue (Variables|VFX|GetBarrelGlowMID) "EmissiveMultiplier" (Variables|VFX|GetBarrelGlowColor(Uncharged)) :!class /Script/Engine.MaterialInstanceDynamic))

(fn ChargedFXOn
  (bind _turbinespinrate_charged (Variables|VFX|GetTurbineSpinRate(Charged)))
  (Transformation|AddLocalRotation (Variables|Default|GetAMRFan) (Math|Rotator|MakeRotator 0.0 _turbinespinrate_charged) false false)
  (Transformation|AddLocalRotation (Variables|Default|GetCoreParticles) (Math|Rotator|MakeRotator _turbinespinrate_charged) false false)
  (Effects|Components|ParticleSystem|SetFloatParameter (Variables|Default|GetCoreParticles) "SpawnRate" (Variables|VFX|GetCoreParticlesSpawnRate(Charged)))
  (Variables|/Script/CascadeRuntime.ParticleSystemComponent|SetCustomTimeDilation (Variables|VFX|GetCoreParticlesTimeDilation(Charged)) (Variables|Default|GetCoreParticles))
  (Effects|Components|ParticleSystem|SetFloatParameter (Variables|Default|GetElectrodeParticles) "SpawnRate" (Variables|VFX|GetElectrodeParticlesEnergizedSpawnRate(Charged)))
  (Rendering|Material|SetVectorParameterValue (Variables|VFX|GetBarrelGlowMID) "EmissiveMultiplier" (Variables|VFX|GetBarrelGlowColor(Charged)) :!class /Script/Engine.MaterialInstanceDynamic))

(fn ConstructionScript
  (|CallParentFunction)
  (Transformation|AttachTo(Deprecated) (Variables|Default|GetElectrodeParticles) (Variables|Visual|GetWeaponMesh) "Barrel_Main" "SnapToTarget" true)
  (Variables|/Script/Engine.SkeletalMeshComponent|SetGlobalAnimRateScale (Variables|VFX|GetAnimationRate(Uncharged)) (Variables|Visual|GetWeaponMesh) :!pinned true) ; Sets the speed of animation components based on charge state
  (Effects|Components|ParticleSystem|SetFloatParameter (Variables|Default|GetCoreParticles) "SpawnRate" (Variables|VFX|GetCoreParticlesSpawnRate(Uncharged)))
  (Variables|/Script/CascadeRuntime.ParticleSystemComponent|SetCustomTimeDilation (Variables|VFX|GetCoreParticlesTimeDilation(Uncharged)) (Variables|Default|GetCoreParticles))
  (Effects|Components|ParticleSystem|SetFloatParameter (Variables|Default|GetElectrodeParticles) "SpawnRate" (Variables|VFX|GetElectrodeParticlesEnergizedSpawnRate(Uncharged)))
  (bind _returnvalue (Rendering|Material|CreateDynamicMaterialInstance (Variables|Default|GetAMRFan) 0 "/Figment_Sniper_AMR_Athena/Gameplay/Weapons/FORT_Rifles/Materials/MI_FORT_Rifles_AMRFan_Inst.MI_FORT_Rifles_AMRFan_Inst" "None" :!class /Script/Engine.PrimitiveComponent))
  (Variables|VFX|SetBarrelGlowMID _returnvalue)
  (Rendering|Material|SetVectorParameterValue (Variables|VFX|GetBarrelGlowMID) "EmissiveMultiplier" (Variables|VFX|GetBarrelGlowColor(Uncharged)) :!class /Script/Engine.MaterialInstanceDynamic)
  (Effects|Components|ParticleSystem|SetFloatParameter (Variables|Default|GetMuzzleLight) "LightBrightness" (Variables|VFX|GetMuzzle%20Flash%20Light%20Intensity%20%28Uncharged%29)))

(fn VFX_SetChargedParams
  (Effects|SetBeamParticleSystem (Variables|VFX|GetChargedBeam))
  (Variables|Effects|Muzzle|SetMuzzle%20System (Variables|VFX|GetMuzzleFlash(Charged)))
  (Effects|Components|ParticleSystem|SetFloatParameter (Variables|Default|GetMuzzleLight) "LightBrightness" (Variables|VFX|GetMuzzle%20Flash%20Light%20Intensity%20%28Charged%29))
  (return))

(fn VFX_SetUnchargedParams
  (Effects|Components|ParticleSystem|SetFloatParameter (Variables|Default|GetMuzzleLight) "LightBrightness" (Variables|VFX|GetMuzzle%20Flash%20Light%20Intensity%20%28Uncharged%29))
  (Variables|Effects|Muzzle|SetMuzzle%20System (Variables|VFX|GetMuzzleFlash(Uncharged)))
  (Effects|SetBeamParticleSystem (Variables|VFX|GetDefaultBeam))
  (return))
