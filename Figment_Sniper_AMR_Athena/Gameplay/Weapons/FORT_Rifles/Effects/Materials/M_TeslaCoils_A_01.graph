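# Unlit masked mesh-particle material for the tesla coil effect on a sniper rifle, with stepped, noise-driven displacement.
# Emissive color comes from particle color or a Color parameter, switched statically, through an emissive adjustment function.
# Two panning noise samples are summed, quantized by a Steps parameter and biased, then scaled along vertex normals.
# Dynamic parameters set thickness and an added offset, while the Morph Speed and Morph Amount channels are unused.
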Material /Figment_Sniper_AMR_Athena/Gameplay/Weapons/FORT_Rifles/Effects/Materials/M_TeslaCoils_A_01
graph {
 properties {bUsedWithMeshParticles=true, blendMode=BLEND_Masked, pixelDepthOffsetMode=PDOM_Legacy, shadingModel=MSM_Unlit, usedShadingModels=Unlit}
 nodes [
  MaterialFunctionCall(Emissive=StaticSwitchParameter(True=ParticleColor().RGB, False=VectorParameter() {parameterName=Color, defaultValue=<rgba 3.0858 2.4847 10 1>}.RGB) {parameterName=Use Particle Color?}) {materialFunction=&MF_FX_EmissiveAdjustments}
  n1 = Panner(Coordinate=TextureCoordinate() {uTiling=0.25, vTiling=0.1}) {speedX=0.12, speedY=0.3}
  n2 = Panner(Coordinate=TextureCoordinate() {uTiling=0.15, vTiling=0.3}) {speedX=-0.08, speedY=0.2}
  n3 = ScalarParameter() {defaultValue=3, parameterName=Steps}
  n4 = Multiply(
   A=Add(
    A=TextureSample(UVs=@n1) {texture=&T_NoisePack_01}.RGB
    B=TextureSample(UVs=@n2) {texture=&T_NoisePack_01}.RGB
   )
   B=@n3
  )
  n5 = ConstantBiasScale(None=Divide(A=Ceil(None=@n4), B=@n3)) {bias=-0.5, scale=1}
  n6 = DynamicParameter() {defaultValue=<rgba 0 0 0 1>, paramNames [Morph Speed, Morph Amount, WPO Thickness, Available]}
  Multiply(A=Add(A=Multiply(A=@n5, B=@n6.WPO Thickness), B=@n6.Available), B=VertexNormalWS())
 ]
}
