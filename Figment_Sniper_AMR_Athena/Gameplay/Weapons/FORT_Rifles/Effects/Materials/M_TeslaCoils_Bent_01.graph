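# Unlit masked mesh-particle material for bent Tesla coil effects on the Figment Sniper AMR rifle.
# Color comes from either particle color or a purple Color parameter via a static switch, feeding MF_FX_EmissiveAdjustments.
# A negated cosine of the V coordinate, biased by an Add, scales a transformed local negative-Y vector into a fixed-magnitude curved offset, apparently for bending the mesh through World Position Offset.
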Material /Figment_Sniper_AMR_Athena/Gameplay/Weapons/FORT_Rifles/Effects/Materials/M_TeslaCoils_Bent_01
graph {
 properties {bUsedWithMeshParticles=true, blendMode=BLEND_Masked, pixelDepthOffsetMode=PDOM_Legacy, shadingModel=MSM_Unlit, usedShadingModels=Unlit}
 nodes [
  MaterialFunctionCall(Emissive=StaticSwitchParameter(True=ParticleColor().RGB, False=VectorParameter() {parameterName=Color, defaultValue=<rgba 3.0858 2.4847 10 1>}.RGB) {parameterName=Use Particle Color?}) {materialFunction=&MF_FX_EmissiveAdjustments}
  n1 = TextureCoordinate()
  n2 = Multiply(A=Cosine(None=ComponentMask(None=@n1) {g=true})) {constB=-1}
  n3 = Constant3Vector() {constant=<rgba 0 -1 0 0>}
  Multiply(A=Multiply(A=Add(A=@n2), B=Transform(None=@n3) {transformSourceType=TRANSFORMSOURCE_Local})) {constB=20}
 ]
}
