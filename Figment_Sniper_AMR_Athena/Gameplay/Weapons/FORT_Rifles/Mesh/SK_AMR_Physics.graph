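# Physics asset for the SK_AMR skeletal mesh, an AMR sniper rifle weapon in the Figment_Sniper_AMR_Athena plugin.
# It holds a single body on the Handle bone, shaped as one long, offset box that stands in for the rifle's collision volume.
# It has no constraints, so the weapon acts as one rigid piece with no separately simulated parts.
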PhysicsAsset /Figment_Sniper_AMR_Athena/Gameplay/Weapons/FORT_Rifles/Mesh/SK_AMR_Physics
bodies {Handle [{shapeType=Box, extentX=129.5421, extentY=24.6663, extentZ=40.907, center=<36.2871 -5.0022 4.6498>}]}
constraints []
