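# Animation Blueprint for the AMR sniper rifle skeleton that adds springy secondary motion to the weapon's cable bones.
# The reference pose feeds DefaultSlot for montages, then passes in component space through chained spring controllers on Cable_Top and Cable_Bottom, the top one stiffer.
# Both springs share a low alpha scaled down from springVal, which nothing here sets, as the event graph holds only unconnected stubs.
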AnimBlueprint /Figment_Sniper_AMR_Athena/Gameplay/Weapons/FORT_Rifles/Mesh/SK_AMR_Skeleton_AnimBP
target_skeleton=&/SK_AMR_Skeleton
graphs [
 AnimGraph {nodes [
   n1 = Local Space Ref Pose()
   n2 = float * float(A=Get springVal() -> (springVal:double), B=.1)
   n3 = Spring controller - Bone: Cable_Top!(ComponentPose:Component Space Pose Link Structure=Local To Component(LocalPose=Slot 'DefaultSlot'(Source=@n1)), Alpha:float=@n2) -> (Pose:Component Space Pose Link Structure) {
    is=title AnimGraphNode_SpringBone
    maxDisplacement=20
    springStiffness=120
    alphaBoolBlend {blendOption=Linear}
    springBone {boneName=Cable_Top}
   }
   Output Pose(Result=Component To Local(ComponentPose=Spring controller - Bone: Cable_Bottom!(ComponentPose:Component Space Pose Link Structure=@n3, Alpha:float=@n2) -> (Pose:Component Space Pose Link Structure) {
      is=title AnimGraphNode_SpringBone
      maxDisplacement=20
      springStiffness=100
      alphaBoolBlend {blendOption=Linear}
      springBone {boneName=Cable_Bottom}
     })) {layerGroup=None}
  ]}
 EventGraph {nodes [
   TryGetPawnOwner()
   Event BlueprintUpdateAnimation()
  ]}
]
variables [springVal {type=number, default=1}]
